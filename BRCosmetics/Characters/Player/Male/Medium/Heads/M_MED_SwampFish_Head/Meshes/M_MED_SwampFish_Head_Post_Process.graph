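# Post-process Animation Blueprint for the SwampFish player head, adding secondary fin motion.
# The input pose runs through four chained Anim Dynamics nodes on the top and bottom left and right fins, using damped angular springs, angular limits and gravity, simulated relative to the root bone.
# The result then passes through the head's Control Rig to the output.
# The event graph only caches the owning skeletal mesh, and the eyelid variable is unused.
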AnimBlueprint /BRCosmetics/Characters/Player/Male/Medium/Heads/M_MED_SwampFish_Head/Meshes/M_MED_SwampFish_Head_Post_Process
target_skeleton=/Game/Characters/Player/Male/Medium/Heads/M_MED_SwampFish_Head/Meshes/CustomSkeleton/M_MED_SwampFish_Head_Skeleton
graphs [
 AnimGraph {nodes [
   n1 = Anim Dynamics - Bone: top_fin_l(ComponentPose=Local To Component(LocalPose=Input Pose()), Alpha=1, ExternalForce="0,0,0") {
    angularDampingOverride=0.9999
    angularSpringConstant=1500
    bAngularSpring=true
    bGravityOverrideInSimSpace=true
    bOverrideAngularDamping=true
    bUseGravityOverride=true
    simulationSpace=BoneRelative
    alphaBoolBlend {blendOption=Linear}
    boundBone {boneName=top_fin_l}
    gravityOverride {z=-980}
    physicsBodyDefinitions [{
      sphereCollisionRadius=0
      boundBone {boneName=top_fin_l}
      boxExtents=<20 20 10>
      constraintSetup {angularTargetAxis=AxisY, angularLimitsMax=<1 15 30>, angularLimitsMin=<-1 -15 -10>, angularTarget=<0 1 0>}
      localJointOffset=<2 0 0>
     }]
    relativeSpaceBone {boneName=root}
   }
   n2 = Anim Dynamics - Bone: bot_fin_l(ComponentPose=@n1, Alpha=1, ExternalForce="0,0,0") {
    angularDampingOverride=0.9999
    angularSpringConstant=1500
    bAngularSpring=true
    bGravityOverrideInSimSpace=true
    bOverrideAngularDamping=true
    bUseGravityOverride=true
    simulationSpace=BoneRelative
    alphaBoolBlend {blendOption=Linear}
    boundBone {boneName=bot_fin_l}
    gravityOverride {z=-980}
    physicsBodyDefinitions [{
      sphereCollisionRadius=0
      boundBone {boneName=bot_fin_l}
      boxExtents=<20 20 10>
      constraintSetup {angularTargetAxis=AxisY, angularLimitsMax=<1 15 30>, angularLimitsMin=<-1 -15 -10>, angularTarget=<0 1 0>}
      localJointOffset=<2 0 0>
     }]
    relativeSpaceBone {boneName=root}
   }
   n3 = Anim Dynamics - Bone: top_fin_r(ComponentPose=@n2, Alpha=1, ExternalForce="0,0,0") {
    angularDampingOverride=0.9999
    angularSpringConstant=1500
    bAngularSpring=true
    bGravityOverrideInSimSpace=true
    bOverrideAngularDamping=true
    bUseGravityOverride=true
    simulationSpace=BoneRelative
    alphaBoolBlend {blendOption=Linear}
    boundBone {boneName=top_fin_r}
    gravityOverride {z=-980}
    physicsBodyDefinitions [{
      sphereCollisionRadius=0
      boundBone {boneName=top_fin_r}
      boxExtents=<15 10 10>
      constraintSetup {angularLimitsMax=<1 10 30>, angularLimitsMin=<-1 -10 -10>, angularTarget=<0 0 0>}
      localJointOffset=<5 0 0>
     }]
    relativeSpaceBone {boneName=root}
   }
   n4 = Anim Dynamics - Bone: bot_fin_r(ComponentPose=@n3, Alpha=1, ExternalForce="0,0,0") {
    angularDampingOverride=0.9999
    angularSpringConstant=1500
    bAngularSpring=true
    bGravityOverrideInSimSpace=true
    bOverrideAngularDamping=true
    bUseGravityOverride=true
    simulationSpace=BoneRelative
    alphaBoolBlend {blendOption=Linear}
    boundBone {boneName=bot_fin_r}
    gravityOverride {z=-980}
    physicsBodyDefinitions [{
      sphereCollisionRadius=0
      boundBone {boneName=bot_fin_r}
      boxExtents=<15 10 10>
      constraintSetup {angularLimitsMax=<1 10 30>, angularLimitsMin=<-1 -10 -10>, angularTarget=<0 0 0>}
      localJointOffset=<5 0 0>
     }]
    relativeSpaceBone {boneName=root}
   }
   n5 = Control Rig(Source=Component To Local(ComponentPose=@n4)) {
    alphaBoolBlend {blendOption=Linear}
    controlRigAssetReference {blueprintRigClass=&/M_MED_SwampFish_Head_CtrlRig.M_MED_SwampFish_Head_CtrlRig_C}
   }
   Output Pose(Result=@n5) {layerGroup=None}
  ]}
 EventGraph {nodes [
   Event BlueprintInitializeAnimation()
   > Set Skel_Mesh(Skel_Mesh=GetOwnerSkeletalMeshComponent())
   TryGetPawnOwner()
   Event BlueprintUpdateAnimation()
  ]}
]
variables [
 eyelid {type=number}
 Skel_Mesh {type=SkeletalMeshComponent}
]
